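# Base squad-info extension widget that shows ToyFox-specific visuals while the player's pawn matches a ToyFox pilot, pawn or seeking-rocket tag query.
# On player state assignment, it binds to pawn changes and restarts paired async tag waits, retrying on a timer until the ability system component is valid.
# Tag results switch the parent CommonVisibilitySwitcher between this widget and index 0.
# Destruct unbinds, clears waits and restores index 0.
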WidgetBlueprint /ToyFoxGameplay/UI/Widgets/Squad/WBP_ToyFox_SquadInfoOverrideBase
graphs [
 ClearAsyncActions {nodes [
   ClearAsyncActions() {is=title K2Node_FunctionEntry}
   > EndAction(self:Ability Async Object Reference=[
     Get AsyncAction_WaitForTagQueryTrue() -> (AsyncAction_WaitForTagQueryTrue:Ability Async Wait Gameplay Tag Query Object Reference)
     Get AsyncAction_WaitForTagQueryFalse() -> (AsyncAction_WaitForTagQueryFalse:Ability Async Wait Gameplay Tag Query Object Reference)
    ])
   > Clear and Invalidate Timer by Handle(Handle=Get RetryTimerHandle() -> (RetryTimerHandle:Timer Handle Structure))
  ]}
 EnableToyFoxVisuals {nodes [
   EnableToyFoxVisuals!() -> (then, IsEnabled:bool=~) {is=title K2Node_FunctionEntry}
   > n1 = Cast To CommonVisibilitySwitcher!(execute, Object=GetParent!(self:Widget Object Reference) -> (ReturnValue:Panel Widget Object Reference)) -> (then, CastFailed, AsCommon Visibility Switcher:Common Visibility Switcher Object Reference) {is=title K2Node_DynamicCast}
   > n2 = Branch(Condition=Get IsEnabled() -> (IsEnabled:bool))
   > SetActiveWidget(self=@n1, Widget=Self-Reference())
   SetActiveWidgetIndex(execute=@n2.else, self:Common Visibility Switcher Object Reference=@n1, Index=0)
  ]}
 EventGraph {nodes [
   n1 = OnRetryPawnSet() {is=title K2Node_CustomEvent}
   Is Valid(exec=[@n8.then, @n1.then], InputObject=Get CurrentPawn() -> (CurrentPawn:Pawn Object Reference))
   > n2 = Is Valid(InputObject=GetAbilitySystemComponent!(Actor:Actor Object Reference=Get CurrentPawn()) -> (ReturnValue:Ability System Component Object Reference))
   n3 = Set Timer by Event(execute=@n2.Is Not Valid, Delegate=Create Event(), Time=0.5, bLooping=false)
   > Set RetryTimerHandle(RetryTimerHandle:Timer Handle Structure=@n3) -> (Output_Get:Timer Handle Structure)
   n4 = Get CommonPlayerGameplayTagQuery() -> (CommonPlayerGameplayTagQuery:Gameplay Tag Query Structure)
   n5 = WaitGameplayTagQueryOnActor!(
    execute:Exec=@n2.Is Valid
    TargetActor:Actor Object Reference=Get CurrentPawn()
    TagQuery:Gameplay Tag Query Structure=@n4
    TriggerCondition:EWaitGameplayTagQueryTriggerCondition Enum=WhenTrue
    bOnlyTriggerOnce:bool=false
   ) -> (then:Exec, AsyncTaskProxy:Ability Async Wait Gameplay Tag Query Object Reference, Triggered:Exec) {is=title K2Node_AsyncAction}
   > Set AsyncAction_WaitForTagQueryTrue(AsyncAction_WaitForTagQueryTrue:Ability Async Wait Gameplay Tag Query Object Reference=@n5) -> (Output_Get:Ability Async Wait Gameplay Tag Query Object Reference)
   > n6 = WaitGameplayTagQueryOnActor(TargetActor=Get CurrentPawn(), TagQuery=@n4, TriggerCondition=WhenFalse, bOnlyTriggerOnce=false)
   > Set AsyncAction_WaitForTagQueryFalse(AsyncAction_WaitForTagQueryFalse:Ability Async Wait Gameplay Tag Query Object Reference=@n6) -> (Output_Get:Ability Async Wait Gameplay Tag Query Object Reference)
   EnableToyFoxVisuals!(execute=@n6.Triggered, self, IsEnabled:bool=false) -> (then) {is=title K2Node_CallFunction}
   EnableToyFoxVisuals(execute=@n5.Triggered, IsEnabled=true)
   n7 = OnPawnSet!() -> (OutputDelegate, then, Player:Player State Object Reference, NewPawn:Pawn Object Reference, OldPawn:Pawn Object Reference) {is=title K2Node_CustomEvent}
   > Set CurrentPawn(CurrentPawn:Pawn Object Reference=@n7.NewPawn) -> (Output_Get:Pawn Object Reference)
   > n8 = ClearAsyncActions() {is=title K2Node_CallFunction}
   Event Destruct()
   > UnbindPlayerState()
   > EnableToyFoxVisuals(IsEnabled=false)
   n9 = Event OnPlayerStateSet!() -> (OutputDelegate, then, InPlayerState:Fort Player State Athena Object Reference) {is=title K2Node_Event}
   > UnbindPlayerState()
   > n10 = Set BoundPlayerState(BoundPlayerState:Fort Player State Athena Object Reference=@n9.InPlayerState) -> (Output_Get:Fort Player State Athena Object Reference)
   > Is Valid(InputObject=@n10)
   > Bind Event to On Pawn Set(self:Player State Object Reference=@n10, Delegate=Create Event()) {is=title K2Node_AddDelegate}
   > OnPawnSet!(
    execute
    self
    Player:Player State Object Reference
    NewPawn:Pawn Object Reference=GetPawn!(self:Player State Object Reference=Get BoundPlayerState() -> (BoundPlayerState:Fort Player State Athena Object Reference)) -> (ReturnValue:Pawn Object Reference)
    OldPawn:Pawn Object Reference
   ) -> (then) {is=title K2Node_CallFunction}
  ]}
 UnbindPlayerState {nodes [
   UnbindPlayerState() {is=title K2Node_FunctionEntry}
   > ClearAsyncActions()
   > n1 = Get!(execute:Exec) -> (then:Exec, else:Exec, BoundPlayerState:Fort Player State Athena Object Reference)
   > Unbind Event from On Pawn Set(self:Player State Object Reference=@n1, Delegate=Create Event()) {is=title K2Node_RemoveDelegate}
   > Set BoundPlayerState()
  ]}
]
variables [
 AsyncAction_WaitForTagQueryFalse {type=AbilityAsync_WaitGameplayTagQuery, default=null}
 AsyncAction_WaitForTagQueryTrue {type=AbilityAsync_WaitGameplayTagQuery, default=null}
 BoundPlayerState {type=FortPlayerStateAthena, default=null}
 CommonPlayerGameplayTagQuery {
  type=GameplayTagQuery
  default {
   autoDescription=" ALL(  ANY( ToyFox.Player.Pilot, ToyFox.Pawn, ToyFox.Ability.Rocket.Seeking.PawnProjectile ) )"
   tokenStreamVersion=0
   userDescription=""
   queryTokenStream [0,1,5,1,1,3,0,1,2]
   tagDictionary [{tagName=ToyFox.Player.Pilot}, {tagName=ToyFox.Pawn}, {tagName=ToyFox.Ability.Rocket.Seeking.PawnProjectile}]
  }
 }
 CurrentPawn {type=Pawn, default=null}
 RetryTimerHandle {type=TimerHandle}
]
widgets {
 info {inheritedWidgetCount=0, parentClass=/Script/FortniteUI.AthenaPlayerInfoBaseExtensionWidget, rootWidgetClass=null, widgetCount=0}
 widgets []
}
