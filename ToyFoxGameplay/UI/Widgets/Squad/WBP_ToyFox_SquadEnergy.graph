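# Squad info widget for ToyFox that shows the Giant power level as a fixed-size progress bar.
# On Construct it finds the ToyFox component on the Athena game state and binds to its Giant power level change event.
# The handler writes the new level into the fill image's dynamic material through the "Progress 02" scalar.
# It sets no initial value, so the bar updates only when that event fires.
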WidgetBlueprint /ToyFoxGameplay/UI/Widgets/Squad/WBP_ToyFox_SquadEnergy
graphs [
 EventGraph {nodes [
   Event Construct()
   > Parent: Construct()
   n1 = GetGameStateAthena!() -> (ReturnValue:Fort Game State Athena Object Reference)
   > Is Valid(InputObject=@n1)
   > n2 = GetComponentByClass!(
    execute
    self:Actor Object Reference=@n1
    ComponentClass:Actor Component Class Reference=/ToyFoxGameplay/Gameplay/B_ToyFoxGameStateComponent.B_ToyFoxGameStateComponent_C
   ) -> (then, ReturnValue:B Toy Fox Game State Component Object Reference)
   > Bind Event to On Giant Power Level Changed(self:B Toy Fox Game State Component Object Reference=@n2, Delegate=Create Event()) {is=title K2Node_AddDelegate}
  ]}
 OnGiantPowerLevelChanged {nodes [
   n1 = OnGiantPowerLevelChanged!() -> (then, PowerLevel:double=~) {is=title K2Node_FunctionEntry}
   > n2 = GetDynamicMaterial(self=Get Image_ProgressFill() -> (Image_ProgressFill:Image Object Reference))
   > SetScalarParameterValue(self=@n2, ParameterName=Progress 02, Value=@n1)
  ]}
]
variables []
widgets {
 info {
  inheritedWidgetCount=1
  parentClass=&/WBP_ToyFox_SquadInfoOverrideBase.WBP_ToyFox_SquadInfoOverrideBase_C
  rootWidgetClass=&SizeBox
  widgetCount=4
 }
 widgets [
  SizeBox_0 {
   widgetClassPath=&SizeBox
   widget=SizeBox_0
   properties {bOverride_HeightOverride=true, bOverride_WidthOverride=true, heightOverride=20, minDesiredWidth=200, widthOverride=200}
  }
  Overlay_0 {widgetClassPath=&Overlay, parent=SizeBox_0, widget=Overlay_0, slot {class=SizeBoxSlot, properties {padding=<ltrb -5 -5 -5 -5>}}}
  Image_ProgressFill {
   widgetClassPath=&Image
   bIsVariable=true
   parent=Overlay_0
   widget=Image_ProgressFill
   properties {brush {
     bIsDynamicallyLoaded=false
     drawAs=Image
     imageType=FullColor
     mirroring=NoMirror
     resourceName=None
     resourceObject=/ToyFoxGameplay/UI/Materials/MedallionBoonsProgress/MI_UI_Boon_Progress
     tiling=NoTile
     imageSize=<24 20>
     margin=<ltrb 0 0 0 0>
     outlineSettings {
      bUseBrushTransparency=false
      roundingType=HalfHeightRadius
      width=0
      color {colorUseRule=UseColor_Specified, specifiedColor=<rgba 0 0 0 0>}
      cornerRadii=<0 0 0 0>
     }
     tintColor {colorUseRule=UseColor_Specified, specifiedColor=<rgba 1 1 1 1>}
     uVRegion {bIsValid=false, max=<0 0>, min=<0 0>}
    }}
   slot {class=OverlaySlot, properties {horizontalAlignment=HAlign_Fill, verticalAlignment=VAlign_Fill}}
  }
  TutorialHighlight {widgetClassPath=&TutorialHighlight}
 ]
}
