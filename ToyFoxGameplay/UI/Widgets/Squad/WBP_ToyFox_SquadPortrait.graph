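# Squad portrait widget for ToyFox, derived from WBP_ToyFox_SquadInfoOverrideBase.
# A fixed-size square SizeBox holds an Overlay with negative padding, so the art bleeds past the box bounds.
# The Overlay stacks two filling material-instance images: a ZordProfile backer beneath a Profile layer named YGradient.
# The event graph is empty and the widget declares no variables beyond the two exposed images, so it contains no logic of its own.
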WidgetBlueprint /ToyFoxGameplay/UI/Widgets/Squad/WBP_ToyFox_SquadPortrait
graphs [EventGraph {links [], nodes []}]
variables []
widgets {
 info {
  inheritedWidgetCount=1
  parentClass=&/WBP_ToyFox_SquadInfoOverrideBase.WBP_ToyFox_SquadInfoOverrideBase_C
  rootWidgetClass=&SizeBox
  widgetCount=5
 }
 widgets [
  SizeBox_0 {
   widgetClassPath=&SizeBox
   widget=SizeBox_0
   properties {bOverride_HeightOverride=true, bOverride_WidthOverride=true, heightOverride=56, widthOverride=56}
  }
  Overlay_31 {
   widgetClassPath=&Overlay
   parent=SizeBox_0
   widget=Overlay_31
   slot {class=SizeBoxSlot, properties {padding=<ltrb -10 -10 -10 -10>}}
  }
  Image_Portrait_Backer {
   widgetClassPath=&Image
   bIsVariable=true
   parent=Overlay_31
   widget=Image_Portrait_Backer
   properties {brush {
     bIsDynamicallyLoaded=false
     drawAs=Image
     imageType=FullColor
     mirroring=NoMirror
     resourceName=None
     resourceObject=/ToyFoxGameplay/UI/Materials/MI_ToyFox_TeamViewer_ZordProfile
     tiling=NoTile
     imageSize=<32 32>
     margin=<ltrb 0 0 0 0>
     outlineSettings {
      bUseBrushTransparency=false
      roundingType=HalfHeightRadius
      width=0
      color {colorUseRule=UseColor_Specified, specifiedColor=<rgba 0 0 0 0>}
      cornerRadii=<0 0 0 0>
     }
     tintColor {colorUseRule=UseColor_Specified, specifiedColor=<rgba 1 1 1 1>}
     uVRegion {bIsValid=false, max=<0 0>, min=<0 0>}
    }}
   slot {class=OverlaySlot, properties {horizontalAlignment=HAlign_Fill, verticalAlignment=VAlign_Fill}}
  }
  Image_Portrait_YGradient {
   widgetClassPath=&Image
   bIsVariable=true
   parent=Overlay_31
   widget=Image_Portrait_YGradient
   properties {brush {
     bIsDynamicallyLoaded=false
     drawAs=Image
     imageType=FullColor
     mirroring=NoMirror
     resourceName=None
     resourceObject=/ToyFoxGameplay/UI/Materials/MI_ToyFox_TeamViewer_Profile
     tiling=NoTile
     imageSize=<32 32>
     margin=<ltrb 0 0 0 0>
     outlineSettings {
      bUseBrushTransparency=false
      roundingType=HalfHeightRadius
      width=0
      color {colorUseRule=UseColor_Specified, specifiedColor=<rgba 0 0 0 0>}
      cornerRadii=<0 0 0 0>
     }
     tintColor {colorUseRule=UseColor_Specified, specifiedColor=<rgba 1 1 1 1>}
     uVRegion {bIsValid=false, max=<0 0>, min=<0 0>}
    }}
   slot {class=OverlaySlot, properties {horizontalAlignment=HAlign_Fill, verticalAlignment=VAlign_Fill}}
  }
  TutorialHighlight {widgetClassPath=&TutorialHighlight}
 ]
}
